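# Niagara idle effect for the SharkFinSwing pickaxe cosmetic, producing intermittent twinkling sparkles around the weapon.
# A single sprite emitter spawns short-lived particles at points picked from a baked position array, avoiding mesh sampling, with a slight camera offset.
# Each sparkle grows and fades in, then out, while its spin ramps up in either direction.
# Warm color and brightness adapt to time of day through FN time-of-day inputs.
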NiagaraSystem /BRCosmetics/Weapons/FORT_Melee/Pickaxe_SharkFinSwing/FX/NS_Pickaxe_SharkFinSwing_Idle
dependencies {
 usedDataInterfaces [&NiagaraDataInterfaceArrayFloat3, &NiagaraDataInterfaceCurve, &NiagaraDataInterfaceArrayFloat]
 usedDynamicInputs [&Multiply_Vector2D_ByFloat]
 usedModules [&SystemState]
}
emitters {Sparkle {
  emitterUpdateScript [
   &EmitterState()
   &SpawnRate(Spawn Probability=0.5, SpawnRate=15, Use Spawn Probability=true)
  ]
  particleSpawnScript [
   &InitializeParticle(
    Color=&FNTimeOfDayColor(
     Night Color=<rgba 1.5 1.2492 0.328 1>
     Morning Color=<rgba 1.5 0.6623 0.15 1>
     Day Color=<rgba 1.5 0.6623 0.15 1>
     Evening Color=<rgba 1.5 0.6623 0.15 1>
    )
    Lifetime=&RandomRangeFloat(Minimum=0.5, Maximum=1)
    Sprite Rotation Angle=0.25
    Sprite Rotation Mode=::NewEnumerator1
    Sprite Size Mode=::NewEnumerator1
    Uniform Sprite Size Max=12.5
    Uniform Sprite Size Min=7.5
    UsePositionOffset=false
    Write Material Random=true
   )
   /SetVariables(Particles.Position=&SelectVectorFromArray(
     Array Sampling Mode=::NewEnumerator0
     Direct Array Index=?
     Vector Selection Array=DataInterface() {propertyValues {FloatData [
        <-1.0835 36.5415 22.1245>
        <2.4738 39.1122 22.9772>
        <0.5628 37.2099 22.6169>
        <-2.9987 37.4975 22.7698>
        <12.7524 18.5971 32.6936>
        <13.3728 12.8796 33.2158>
        <14.1568 11.3493 32.7451>
        <13.06 9.0714 33.3667>
        <-11.9018 9.6616 32.4127>
        <-14.221 15.6676 31.9382>
        <-12.6065 18.3981 33.9384>
        <-14.2931 13.6023 32.7715>
        <8.4734 -10.1338 30.3274>
        <-7.2643 -7.2305 32.1039>
        <4.8022 40.3997 27.7019>
        <6.8709 34.5805 30.3068>
        <3.2323 34.5942 23.2714>
        <3.4331 36.4219 23.8164>
        <-8.8003 33.5992 26.6115>
        <-9.7882 31.2316 28.8697>
        <-4.8239 43.1038 26.2476>
        <-7.6785 37.7589 29.0847>
        <-9.6046 31.08 29.6721>
        <-8.8888 -28.4553 24.9141>
        <-5.8569 -28.764 26.0383>
        <-4.2511 -26.8431 28.1123>
        <-5.6182 -25.6073 28.3068>
        <-5.5064 -23.7878 28.675>
        <-4.9998 41.936 27.3771>
        <-4.6894 28.0473 21.5493>
        <-5.3338 23.2525 23.6307>
        <-2.3051 17.5951 24.4784>
        <-3.5112 16.9175 24.8574>
        <-0.2016 7.7698 25.0161>
        <-6.8179 17.525 26.062>
        <-6.6722 11.3871 26.6307>
        <-9.5198 16.1795 28.0944>
        <-7.7635 20.1537 26.0184>
        <-8.7295 25.8053 24.5846>
        <-10.4934 28.48 26.065>
        <-9.4799 29.321 24.0215>
        <-7.1461 32.8354 22.143>
        <-11.4187 20.0272 29.9389>
        <-6.1682 1.9761 27.2204>
        <-6.4632 -0.2622 28.45>
        <-1.4439 -6.6374 26.8305>
        <-3.8217 -6.6087 28.1312>
        <-6.1053 -7.5948 31.9088>
        <-3.4821 -13.5146 29.7264>
        <-2.3528 -15.3814 28.7932>
        <-2.742 -15.9261 29.2966>
        <-0.286 -27.8561 26.7952>
        <-0.1835 -36.5259 39.5613>
        <-0.3708 -22.1633 34.2571>
        <-3.3057 -15.4766 30.0009>
        <-4.1283 -14.0465 32.0095>
        <-3.8168 -14.7088 33.8821>
        <-1.8102 -14.7494 36.2643>
        <-0.64 -12.291 37.7572>
        <-6.1148 -4.1471 38.5663>
        <-10.2267 3.8343 33.0643>
        <-12.0381 11.5912 36.2495>
        <-12.4361 15.1774 35.1002>
        <-11.8858 14.8855 37.689>
        <-10.1163 13.3438 40.5286>
        <-7.0799 10.9461 42.69>
        <-4.9948 11.7226 43.8507>
        <-6.0648 5.3955 42.1073>
        <-0.6846 -5.5377 40.7593>
        <-1.2352 14.0827 45.8026>
        <-0.9622 15.5657 48.1836>
        <-1.0481 18.4448 47.4687>
        <-0.154 11.3323 53.1866>
        <-0.176 13.1074 57.7898>
        <-0.4062 15.655 54.7049>
        <-0.3458 24.725 49.5055>
        <-0.6305 21.3617 49.605>
        <-3.9721 22.0224 44.5264>
        <-5.7375 26.0999 43.0481>
        <-4.5827 36.8035 39.7288>
        <-4.263 43.8626 26.7523>
        <-5.7063 41.0812 27.8033>
        <-9.5415 33.8558 31.3616>
        <-10.1933 33.0145 32.0345>
        <-9.5706 35.303 32.531>
        <-5.9678 40.9684 35.4383>
        <-8.0817 39.163 35.1103>
        <-9.3341 34.9595 34.6935>
        <-6.523 34.8559 39.5124>
        <-8.2441 30.8287 39.5527>
        <-10.8025 27.976 36.4031>
        <-10.9408 26.5504 37.2788>
        <-8.6962 28.1549 40.2049>
        <-10.3695 25.1802 38.7652>
        <-7.5779 24.2747 42.5174>
        <3.0549 27.0413 21.6275>
        <-0.0399 21.4843 23.128>
        <3.3758 21.6782 23.6564>
        <-0.0213 11.9713 24.9767>
        <1.8699 13.2414 25.1053>
        <1.2912 6.5651 25.204>
        <3.574 9.3214 25.6093>
        <7.8429 12.9734 27.4027>
        <5.997 19.9481 25.152>
        <7.9713 20.7443 26.1726>
        <10.2655 24.255 27.897>
        <10.5742 21.3122 28.8402>
        <6.6929 32.8421 22.83>
        <6.8648 31.1236 27.0807>
        <8.6022 31.4366 30.986>
        <8.7677 34.3972 31.3448>
        <5.0815 44.0216 28.3283>
        <0.5417 49.2011 26.2592>
        <-0.244 54.272 28.0717>
        <-0.7016 52.4625 30.4153>
        <2.5926 49.8167 27.5475>
        <4.2595 46.3159 30.0631>
        <4.1966 45.6047 31.2556>
        <0.4419 43.1071 36.6437>
        <4.2316 40.8704 36.2259>
        <0.2121 24.2983 51.3181>
        <0.3986 25.8976 48.7038>
        <0.9056 24.3663 46.0106>
        <1.375 24.2711 45.061>
        <0.7808 14.5694 54.2178>
        <0.58 11.4333 58.2887>
        <0.562 11.7079 52.6602>
        <0.6551 12.5803 51.8338>
        <0.8758 14.5666 50.5959>
        <0.9897 12.0341 45.4054>
        <0.668 -1.037 42.3418>
        <1.1377 -2.405 41.755>
        <0.145 -7.939 41.9432>
        <6.3958 1.0326 40.4924>
        <5.3809 -9.5417 34.4854>
        <0.0978 -28.6807 37.8196>
        <0.3184 -27.7782 35.8956>
        <0.1392 -31.6356 39.4708>
        <0.4013 -35.5583 40.2839>
        <-0.2544 -30.2872 27.219>
        <0.6567 -16.3877 27.8056>
        <0.0652 -16.4482 27.5808>
        <1.4406 -15.2205 28.2021>
        <2.831 -9.5722 28.3309>
        <6.215 -6.4776 31.9039>
        <2.339 -18.8403 30.2159>
        <2.7911 -18.1197 31.3087>
        <4.8403 -0.7632 27.1508>
        <3.1761 4.5723 25.5965>
        <6.5178 6.3957 26.9257>
        <6.7275 0.3713 28.7473>
        <11.6021 13.8575 32.008>
        <11.0206 24.09 30.245>
        <10.6784 28.6153 32.0278>
        <10.0043 29.2888 35.4518>
        <7.9439 37.1071 31.5088>
        <7.4371 38.3893 30.599>
        <7.368 38.7277 32.3051>
        <7.0848 36.7551 36.0787>
        <7.5124 31.3266 39.1239>
        <3.1529 18.8505 44.7785>
        <8.163 21.5633 42.0112>
        <5.9015 16.8241 43.5991>
        <10.8719 18.0487 39.3612>
        <7.3102 17.1946 42.8858>
        <8.2171 16.1181 42.1539>
        <5.4571 9.6398 43.2212>
        <7.6141 11.7222 42.2823>
        <7.7537 7.7832 41.331>
        <9.1919 1.0287 34.1804>
        <2.3 -23.2239 30.2334>
        <1.3073 -25.185 27.6436>
        <7.4058 31.9111 30.8297>
        <7.1216 29.4419 29.3647>
        <5.6789 28.7005 28.7461>
        <0.8132 29.4562 24.8763>
        <-0.5267 28.7309 24.9928>
        <1.8228 24.9726 28.2268>
        <-1.5457 38.1269 31.4757>
        <3.8915 30.57 32.5517>
        <-3.4311 31.9903 33.003>
        <5.2447 25.5494 32.7523>
        <1.3877 29.5884 34.2655>
        <-1.0326 30.3677 34.3367>
        <4.5719 22.2647 32.6864>
        <0.3618 21.7672 36.2157>
        <-1.1217 26.8022 35.184>
        <0.3617 17.0767 36.1642>
        <-5.5111 23.6698 33.656>
        <-3.4266 18.5612 35.8516>
        <-5.5294 23.3785 33.4526>
        <0.148 18.7086 32.3567>
       ]}}
     Random Seed=0
    )) {isSetParameters=true}
   &CameraOffset(Camera Offset Amount=1.5)
  ]
  particleUpdateScript [
   &ParticleState()
   &ScaleSpriteSize(Uniform Curve Sprite Scale=DataInterface() {propertyValues {Curve {keys [
        {arriveTangent=4.7337, interpMode=RCIM_Cubic, leaveTangent=4.7337, tangentMode=RCTM_User, time=0, value=0}
        {interpMode=RCIM_Cubic, tangentMode=RCTM_User, time=0.5, value=1}
        {tangentMode=RCTM_User, time=1, value=0}
       ]}}})
   &ScaleColor(
    Scale Alpha=&FloatFromCurve(FloatCurve=DataInterface() {propertyValues {Curve {keys [
         {arriveTangent=4.5234, interpMode=RCIM_Cubic, leaveTangent=4.5234, tangentMode=RCTM_User, time=0, value=0}
         {interpMode=RCIM_Cubic, tangentMode=RCTM_User, time=0.5, value=1}
         {tangentMode=RCTM_User, time=1, value=0}
        ]}}})
    Scale RGB=&VectorFromFloat(Value=&FNTimeOfDayScalar(Night Scale=0.5, Morning Scale=1, Evening Scale=0.8))
    ScaleA=true
    ScaleRGB=true
   )
   &SpriteRotationRate(Rotation Rate=&FloatFromCurve(
     FloatCurve=DataInterface() {propertyValues {Curve {keys [
         {interpMode=RCIM_Cubic, tangentMode=RCTM_User, time=0.2, value=0}
         {arriveTangent=3.4687, interpMode=RCIM_Cubic, leaveTangent=3.4687, tangentMode=RCTM_User, time=1, value=1}
        ]}}}
     Scale Curve=&SelectFloatFromArray(
      Array Sampling Mode=::NewEnumerator0
      Direct Array Index=?
      Float Selection Array=DataInterface() {propertyValues {FloatData [-800,800]}}
      Random Seed=0
     )
    ))
  ]
  rendererClasses [&NiagaraSpriteRendererProperties]
  renderers [{rendererClass=&NiagaraSpriteRendererProperties}]
 }}
summary {
 systemName=NS_Pickaxe_SharkFinSwing_Idle
 emitters [{emitterName=Sparkle, rendererClasses [&NiagaraSpriteRendererProperties]}]
 userVariables [
  User.bIsFrontEnd {type=NiagaraBool}
  User.bIsFrontEndPreview {type=NiagaraBool}
 ]
}
